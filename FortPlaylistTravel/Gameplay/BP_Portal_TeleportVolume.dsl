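; Interactable portal volume that broadcasts a request to show the playlist travel initiator, letting the local player travel to another experience.
; BeginPlay caches eligibility from the portal CVar, split-screen state and fill rules, keeping the per-frame interact check cheap.
; A safe-zone phase listener blocks interaction once a CVar-set phase is reached, to prevent cheating.
; Only first interactions trigger travel, leaving second interactions for subclasses.
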(event EventBeginPlay
  (Utilities|FlowControl|Sequence
    (:then_0
      ; Cache if the portal should be usable or not so BlueprintCanInteract is more performant as it's called every frame when the player is close to the portal.
      (bind _shouldenable (CallFunction|ShouldEnablePortal))
      (Variables|Default|SetbCachedShouldEnablePortal _shouldenable))
    (:then_1
      ; Disable the portal after CVarSafeZoneIndexDisable index to prevent cheating
      (bind _gamestate (Class|FortAsyncActionCurrentPlaylistReady|PlaylistDataReadyAsync)
        (:OnReady
          (bind _returnvalue (Development|GetConsoleVariableIntValue (Variables|Default|GetCVarSafeZoneIndexDisable)))
          (if (> _returnvalue 0)
            (bind _returnvalue_2 (GameplayEventRouter|GetEventRouter "/Script/FortniteGame.FortGameState" _gamestate))
            (Utilities|IsValid _returnvalue_2
              (:Is%20Valid
                (bind (_asynctaskproxy _payload) (GameplayEventRouter|StartListeningToStatefulEvent _returnvalue_2 "/Script/FortniteGame.FortSafeZonePhaseUpdatedEvent")
                  (:OnEventReceived
                    (bind _returnvalue_3 (Development|GetConsoleVariableIntValue (Variables|Default|GetCVarSafeZoneIndexDisable)))
                    (if (>= (Utilities|Struct|BreakFortSafeZonePhaseUpdatedEvent _payload) _returnvalue_3)
                      (Variables|Default|SetbSafeZonesAllowPortal false)
                      (Fortnite|FortShippingLog 0 (Utilities|String|Append "BP_Portal_TeleportVolume: Disable portal,  safe zone phase index = " (Utilities|String|ToString(Integer) (Utilities|Struct|BreakFortSafeZonePhaseUpdatedEvent _payload)))))))))))))))

(fn BlueprintCanInteract
  (params (InteractingPawn :type Fort%20Pawn%20Object%20Reference) (InteractionBeingAttempted :type EInteractionBeingAttempted%20Enum :default "FirstInteraction") (InteractionType :type TInteractionType%20Enum :default "IT_NoInteraction"))
  (return (and (Variables|Default|GetbCachedShouldEnablePortal) (Variables|Default|GetbSafeZonesAllowPortal))))

(fn BlueprintGetInteractionString
  (params (InteractingPawn :type Fort%20Pawn%20Object%20Reference) (InteractionBeingAttempted :type EInteractionBeingAttempted%20Enum :default "FirstInteraction"))
  (return (Variables|Default|GetInteractString)))

(fn BlueprintOnLocalInteract
  (params (InteractingPawn :type Fort%20Player%20Pawn%20Object%20Reference) (InteractionBeingAttempted :type EInteractionBeingAttempted%20Enum :default "FirstInteraction"))
  (bind _returnvalue (GameplayEventRouter|GetGlobalEventRouter))
  (bind _returnvalue_2 (Fortnite|GetFortPlayerControllerfromActor InteractingPawn))
  ;[0 0 0 0.5] Allow subclasses to override the
  ; second interaction if desired
  (if (Utilities|Enum|Equal(Enum) InteractionBeingAttempted "FirstInteraction")
    (bind _returnvalue_3 (Development|GetConsoleVariableBoolValue (Variables|Default|GetPortalEnabledCVarString)))
    (if (or _returnvalue_3 (Utilities|String|IsEmpty (Variables|Default|GetPortalEnabledCVarString)))
      (Development|PrintString "TELEPORT VOLUME TO ANOTHER EXPERIENCE")
      (Utilities|IsValid _returnvalue_2
        (:Is%20Valid
          (Utilities|IsValid _returnvalue
            (:Is%20Valid
              (GameplayEventRouter|BroadcastStatefulEvent _returnvalue (Utilities|Struct|MakeFortPlaylistTravelEventRequestShowTravelInitiator (Variables|Default|GetInitiatorOptions)) _returnvalue_2)
              (return false)))))
      (else
        (Development|PrintString "Can't teleport, CVAR off")))))

(fn IsFillPlaylist
  (bind _returnvalue (Class|SpecialEventGameplayLibrary|IsFillEnabledforLocalParty))
  (if _returnvalue
    (return true)
    (else
      (return (not (GameplayTags|HasAnyTags (Class|FortGameStateAthena|GetAthenaPlaylistContextTags (Fortnite|Athena|GameState|GetGameStateAthena)) (Variables|Default|GetNoFillPlaylistTags) true))))))

(fn ShouldEnablePortal
  ;[1 1 1] If disabled by CVar, don't allow interaction
  (bind _returnvalue (Development|GetConsoleVariableBoolValue (Variables|Default|GetPortalEnabledCVarString)))
  (if (or _returnvalue (Utilities|String|IsEmpty (Variables|Default|GetPortalEnabledCVarString)))
    ;[0.3712 0.3712 0.3712] Don't allow interaction for Split Screen users.
    (if (Viewport|HasMultipleLocalPlayers)
      (return false)
      (else
        ;[1 1 1] If EnableOnFill CVar always on, or if we are playing on no fill, allow interaction. Otherwise, only allow travel on no fill playlists
        (bind _returnvalue_2 (Development|GetConsoleVariableBoolValue (Variables|Default|GetCVarBRPortalEnableOnFill)))
        (if _returnvalue_2
          (return true)
          (else
            (bind _isfill (CallFunction|IsFillPlaylist))
            (return (not _isfill))))))
    (else
      (return false))))

(fn ConstructionScript)
